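# Material function that adds aim-down-sights scope behavior to weapon materials, behind the "ADS Enable" static switch.
# It feeds the world-space normal into AttachScope and writes back its world position offset and its normal, converted to tangent space.
# AttachScope opacity is sharpened by power, multiply and clamp, then dithered through DitherTemporalAA.
# When disabled, input attributes pass through unchanged; the WeaponSightCameraOffsetManual parameter is declared but unconnected here.
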MaterialFunction /Game/Packages/Fortnite_UniqueMaterials_1/Weapons/ADSScopeMaster
graph {nodes [
  n1 = FunctionInput() {inputName=Material Attributes, inputType=FunctionInput_MaterialAttributes}
  n2 = GetMaterialAttributes(None=@n1) {attributeGetTypes [#1]}
  n3 = MaterialFunctionCall!(RotateDialMask, Normal WS=Transform(None=@n2.Normal)) {materialFunction=/Game/Packages/Fortress_SharedMaterials/Base_Material_Functions/AttachScope}
  n4 = Power(Base=@n3.Opacity) {constExponent=3}
  n5 = MaterialFunctionCall(Alpha Threshold=Clamp(None=Multiply(A=@n4) {constB=20})) {materialFunction=&DitherTemporalAA}
  n6 = Transform(None=@n3.World Normal) {transformSourceType=TRANSFORMSOURCE_World, transformType=TRANSFORM_Tangent}
  FunctionOutput(None=StaticSwitchParameter(
    True=SetMaterialAttributes!(MaterialAttributes=@n2.MaterialAttributes, Opacity=@n5.Result, Normal=@n6, WorldPositionOffset=@n3.World Position Offset) {attributeSetTypes [#2, #1, #3]}
    False=@n1
   ) {group=ADS, parameterName=ADS Enable})
  ScalarParameter() {group=ADS, parameterName=WeaponSightCameraOffsetManual}
 ]}
