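# Base weapon material function that adds optional screen-attached scope behavior to opacity mask, normal and world position offset.
# A static switch, IsScreenAttachedWeaponScope?, selects AttachScope results, converting its world normal to tangent space, or passes the inputs through.
# In scope mode, AttachScope opacity is sharpened, clamped and run through DitherTemporalAA to modulate the incoming opacity.
# Diffuse Alpha multiplies the opacity mask in both modes.
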MaterialFunction /Game/Packages/Fortress_SharedMaterials/Base_Material_Functions/BaseCustomizationFunctions/MF_BaseWeapon_BasicScope
graph {nodes [
  n1 = FunctionInput() {inputName=Normal, sortPriority=7}
  n2 = MaterialFunctionCall!(RotateDialMask, Normal WS=@n1) {materialFunction=/Game/Packages/Fortress_SharedMaterials/Base_Material_Functions/AttachScope}
  n3 = Multiply(A=Power(Base=@n2.Opacity) {constExponent=3}) {constB=20}
  n4 = FunctionInput() {bUsePreviewValueAsDefault=true, inputName=Modulative Opacity, sortPriority=5}
  n5 = Multiply(
   A=@n4
   B=MaterialFunctionCall(Alpha Threshold=Clamp(None=@n3)) {materialFunction=&DitherTemporalAA}.Result
  )
  n6 = StaticBoolParameter() {group=Scope, parameterName=IsScreenAttachedWeaponScope?}
  FunctionOutput(None=Multiply(
    A=StaticSwitch(True=@n5, False=@n4, Value=@n6)
    B=FunctionInput() {bUsePreviewValueAsDefault=true, inputName=Diffuse Alpha, inputType=FunctionInput_Scalar, sortPriority=6}
   )) {outputName=Opacity Mask, sortPriority=5}
  FunctionOutput(None=StaticSwitch(True=Transform(None=@n2.World Normal) {transformSourceType=TRANSFORMSOURCE_World, transformType=TRANSFORM_Tangent}, False=@n1, Value=@n6)) {outputName=Normal, sortPriority=6}
  FunctionOutput(None=StaticSwitch(True=@n2.World Position Offset, False=FunctionInput() {bUsePreviewValueAsDefault=true, inputName=WPO, sortPriority=8}, Value=@n6)) {outputName=Vertex Opacity WPO, sortPriority=7}
 ]}
