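# Material function that builds the World Position Offset for base weapon materials, combining an LOD effect with a black hole distortion.
# LODEffect, driven by the pickup distance-cull override, yields an offset that is added to world position and fed into BlackHoleFunction along with the black hole position parameter.
# The black hole result plus the LOD offset forms the WPO output.
# Mobile shading and Low quality output a constant zero, skipping both effects.
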MaterialFunction /Game/Packages/Fortress_SharedMaterials/Base_Material_Functions/BaseCustomizationFunctions/MF_BaseWeapon_WPO
graph {nodes [
  n1 = VectorParameter() {group=Weapon_WorldPositionOffset, parameterName=WorldPostionOfBlackHole, defaultValue=<rgba -30000 25536 1952 0>}
  n2 = AppendVector(A=@n1.RGB, B=@n1.RGB)
  n3 = MaterialFunctionCall!(
   World Position
   DistanceCullFadeAnimationOffset=MaterialFunctionCall() {materialFunction=/Game/Packages/Fortress_SharedMaterials/Base_Material_Functions/PickupEffectDistanceCullOverride}.Result
   Use actor center instead of pivot point
  ) {materialFunction=/Game/Packages/Fortress_SharedMaterials/Base_Material_Functions/LODEffect}
  n4 = Add(A=@n3.Result, B=WorldPosition().XYZ)
  n5 = Constant()
  n6 = ShadingPathSwitch(
   Default=Add(
    A=MaterialFunctionCall!(WorldPostionOfBlackHole=@n2, World Position=@n4) {materialFunction=/Game/Packages/Fortnite_UniqueMaterials_1/BlackHoleFunction}.Result
    B=@n3.Result
   )
   Mobile=@n5
  ) {desc=Migrated from FeatureLevelSwitch}
  FunctionOutput(None=QualitySwitch(Default=@n6, Low=@n5, High=@n6, Medium=@n6)) {description=Part of Black Hole Function, outputName=WPO, sortPriority=2}
 ]}
