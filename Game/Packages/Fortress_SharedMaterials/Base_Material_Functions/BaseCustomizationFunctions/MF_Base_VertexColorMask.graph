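# MF_Base_VertexColorMask is a shared material function that provides an optional vertex-color-driven mask for base materials.
# A static switch parameter, "Use Vertex Colors for Mask" in the Opacity Mask group, selects the output.
# When on, it returns the result of the Fortnite_HideElementsWithVertexColor function, used to hide elements by vertex color; when off, it returns a constant.
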MaterialFunction /Game/Packages/Fortress_SharedMaterials/Base_Material_Functions/BaseCustomizationFunctions/MF_Base_VertexColorMask
graph {nodes [FunctionOutput(None=StaticSwitchParameter(
    True=MaterialFunctionCall!(WPO) {materialFunction=/Game/Packages/Fortress_SharedMaterials/Base_Material_Functions/Fortnite_HideElementsWithVertexColor}.Result
    False=Constant()
   ) {group=Opacity Mask, parameterName=Use Vertex Colors for Mask})]}
