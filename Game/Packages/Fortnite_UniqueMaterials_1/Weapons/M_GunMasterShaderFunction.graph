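# Master material function for Fortnite guns that assembles a full Material Attributes result and exposes individual channels and helper masks.
# MF_BaseWeapon supplies surface channels and clear coat mask, a separate function supplies emissive, and BasicScope processes rez-in opacity, normal and WPO.
# Further outputs provide customization masks, pre-skinned position and normal, a muzzle sphere mask and a random float.
# It also decodes smooth normals from vertex color and builds a decal mask from UV1.
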MaterialFunction /Game/Packages/Fortnite_UniqueMaterials_1/Weapons/M_GunMasterShaderFunction
graph {
 properties {bExposeToLibrary=true, libraryCategoriesText [Misc,Fortnite]}
 nodes [
  n1 = FunctionInput(Preview=TextureCoordinate()) {bUsePreviewValueAsDefault=true, inputName=UV 1 coordinates, inputType=FunctionInput_Vector2, sortPriority=2}
  n2 = MaterialFunctionCall!(UV 1 Coordinates=@n1) {materialFunction=/Game/Packages/Fortress_SharedMaterials/Base_Material_Functions/BaseCustomizationFunctions/MF_Base_CustomizationMask}
  FunctionOutput(None=@n2."Customization Mask (_CM RGB)") {outputName="Customization Mask (_CM RGB)", sortPriority=31}
  FunctionOutput(None=@n2."Customization Mask 1 (_CM R)") {outputName="Customization Mask 1 (_CM B Top Left)", sortPriority=33}
  n3 = TextureCoordinate() {coordinateIndex=1, desc=Decal position}
  FunctionOutput(None=@n3) {outputName=Decal UVs, sortPriority=46}
  FunctionOutput(None=@n2."Customization Mask 2 (_CM B)") {outputName="Customization Mask 2 (_CM B Top Right)", sortPriority=34}
  FunctionOutput(None=MaterialFunctionCall() {materialFunction=/Game/Packages/Fortress_SharedMaterials/Base_Material_Functions/BaseCustomizationFunctions/MF_BaseWeapon_RandomFloat}.randomFloat) {outputName=randomFloat, sortPriority=48}
  n4 = MaterialFunctionCall() {materialFunction=/Game/Packages/Fortress_SharedMaterials/Base_Material_Functions/BaseCustomizationFunctions/MF_Base_LocalPosition}
  n5 = MaterialFunctionCall!(
   MuzzleFalloff=FunctionInput(Preview=ScalarParameter() {defaultValue=50, group=Weapon_Muzzle, parameterName=MuzzleFalloff}) {bUsePreviewValueAsDefault=true, inputName=MuzzleFalloff, inputType=FunctionInput_Scalar, sortPriority=4}
   Pre-Skinned Local Position=FunctionOutput(None=@n4.Pre-Skinned Local Position Pixel Shader) {description=Local Position Pixel Shader, outputName=Pre-Skinned Local Position Pixel Shader, sortPriority=42}
  ) {materialFunction=/Game/Packages/Fortress_SharedMaterials/Base_Material_Functions/BaseCustomizationFunctions/MF_BaseWeapon_MuzzleSphereMask}
  FunctionOutput(None=@n5.Muzzle_LocalPosition) {outputName=Muzzle_LocalPosition, sortPriority=43}
  FunctionOutput(None=Constant()) {outputName=-------------------, sortPriority=17}
  FunctionOutput(None=Constant()) {outputName=--------------------, sortPriority=30}
  FunctionOutput(None=@n2.Customization Mask UV Islands) {outputName="Customization Mask UV Islands (_CM B Bottom Right)", sortPriority=36}
  FunctionOutput(None=Constant()) {outputName=---------------------, sortPriority=40}
  FunctionOutput(None=Constant()) {outputName=------------------, sortPriority=-19}
  FunctionOutput(None=@n2."Customization Crevasse (_CM G)") {outputName="Customization Crevasse (_CM G)", sortPriority=32}
  n6 = MaterialFunctionCall() {materialFunction=/Game/Packages/Fortress_SharedMaterials/Base_Material_Functions/BaseCustomizationFunctions/MF_BaseWeapon_Emissive}
  FunctionOutput(None=@n6.Cropped Emissive Texture) {description=Passthrough for Customization, outputName=Cropped Emissive Mask, sortPriority=5}
  FunctionOutput(None=@n4.Pre-Skinned Local Normal Pixel Shader) {description=Local Position Pixel Shader, outputName=Pre-Skinned Local Normal Pixel Shader, sortPriority=42}
  FunctionOutput(None=@n2."Customization GMAP (_CM A)") {outputName=Customization Gmap, sortPriority=31}
  FunctionOutput(None=@n2."Customization Mask (_CM RGBA)") {outputName="Customization Mask (_CM RGBA)", sortPriority=31}
  n7 = ConstantBiasScale(None=VertexColor()) {bias=-0.5, scale=2}
  FunctionOutput(None=Transform(None=Normalize(VectorInput=@n7)) {transformSourceType=TRANSFORMSOURCE_Local}) {outputName=Smooth Normal, sortPriority=45}
  n8 = MaterialFunctionCall!(UV 1 Coordinates=@n1) {materialFunction=/Game/Packages/Fortress_SharedMaterials/Base_Material_Functions/BaseCustomizationFunctions/MF_BaseWeapon}
  n9 = MaterialFunctionCall!(
   Modulative Opacity=FunctionOutput(None=MaterialFunctionCall!(Animation Gradient, Rez In Percentage) {materialFunction=/Game/Packages/Fortress_SharedMaterials/Base_Material_Functions/CharacterResInEffect}.Modulative Opacity) {outputName=Opacity Mask, sortPriority=7}
   Diffuse Alpha=@n8.Diffuse Alpha
   Normal=@n8.Diffuse Alpha
   WPO=MaterialFunctionCall() {materialFunction=/Game/Packages/Fortress_SharedMaterials/Base_Material_Functions/BaseCustomizationFunctions/MF_BaseWeapon_WPO}.WPO
  ) {materialFunction=/Game/Packages/Fortress_SharedMaterials/Base_Material_Functions/BaseCustomizationFunctions/MF_BaseWeapon_BasicScope}
  n10 = ScalarParameter() {defaultValue=1, group=Weapon_ClearCoat, parameterName=Clear Coat Opacity, sliderMax=1}
  FunctionOutput(None=MakeMaterialAttributes!(
    BaseColor=FunctionOutput(None=@n8.Diffuse) {outputName=Diffuse, sortPriority=1}
    Metallic=FunctionOutput(None=@n8."Metalness (_S G)") {outputName="Metalness (_S G)", sortPriority=2}
    Specular=FunctionOutput(None=@n8."Specular (_S R)") {outputName="Specular (_S R)", sortPriority=3}
    Roughness=FunctionOutput(None=@n8."Roughness (_S B)") {outputName="Roughness (_S B)", sortPriority=4}
    Anisotropy
    EmissiveColor=@n6.Result
    Opacity=Constant()
    OpacityMask=@n9.Opacity Mask
    Normal=FunctionOutput(None=@n9.Normal) {outputName=Normal Texture, sortPriority=8}
    Tangent
    WorldPositionOffset=FunctionOutput(None=@n9.Vertex Opacity WPO) {outputName=World Position Offset, sortPriority=41}
    SubsurfaceColor=Constant3Vector()
    ClearCoat=Multiply(A=@n8."CC (_M B)", B=@n10)
    ClearCoatRoughness=ScalarParameter() {group=Weapon_ClearCoat, parameterName=ClearCoatRoughness}
    AmbientOcclusion
    Refraction
    CustomizedUVs_0
    CustomizedUVs_1
    CustomizedUVs_2
    CustomizedUVs_3
    CustomizedUVs_4
    CustomizedUVs_5
    CustomizedUVs_6
    CustomizedUVs_7
    PixelDepthOffset
    ShadingModel
    Displacement
   )) {outputName=Material Result, sortPriority=-200}
  FunctionOutput(None=@n2."Customization Mask 3 (_CM G)") {outputName="Customization Mask 3 (_CM B Bottom Left)", sortPriority=35}
  FunctionOutput(None=@n5.Muzzle_SphereMask) {outputName=MuzzleSphereMask, sortPriority=44}
  n11 = MaterialFunctionCall(Float2=@n3) {materialFunction=&BreakOutFloat2Components}
  n12 = Constant() {r=0.5}
  FunctionOutput(None=Multiply(A=SphereMask(A=@n11.R, B=@n12, Radius=@n12), B=SphereMask(A=@n11.G, B=@n12, Radius=@n12))) {outputName=Decal Mask, sortPriority=47}
  Multiply(
   A=@n9.Vertex Opacity WPO
   B=MaterialFunctionCall() {materialFunction=/Game/Packages/Fortress_SharedMaterials/Base_Material_Functions/BaseCustomizationFunctions/MF_Base_VertexColorMask}.Result
  ) {desc="Bypassed as part of "}
  Reroute(None=FunctionOutput() {description=Weapon Decals, outputName=Custom Tex Coord 2, sortPriority=20})
  Reroute(None=FunctionOutput() {description=Weapon Top Down Projection, outputName=Custom Tex Coord 3, sortPriority=21})
  Reroute(None=FunctionOutput() {description=Weapon Side-Side Projection, outputName=Custom Tex Coord 4, sortPriority=22})
 ]
}
