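# Material function that supplies weapon local-space position and normal from the pre-skinned mesh, for use in vertex and pixel shaders.
# Vertex outputs pass pre-skinned position and normal directly, and also pack their six components into three float2 outputs for custom texture coordinates 2–4.
# Pixel outputs rebuild both vectors from those UV channels, with channel 3 shared between position Z and normal X.
# The pixel outputs rely on the material wiring those custom UV outputs.
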MaterialFunction /Game/Packages/Fortnite_UniqueMaterials_1/Weapons/WeaponLocalPosition
graph {
 properties {bExposeToLibrary=true, description="Weapon Local Space Position (Script Based)"}
 nodes [
  AppendVector(
   A=MaterialFunctionCall(Float2=TextureCoordinate() {coordinateIndex=3}) {materialFunction=&BreakOutFloat2Components}.G
   B=TextureCoordinate() {coordinateIndex=4}
  )
  n1 = PreSkinnedPosition()
  FunctionOutput(None=@n1) {outputName=Weapon Local Position Vertex Shader, sortPriority=200}
  FunctionOutput(None=ComponentMask(None=@n1) {g=true, r=true}) {outputName=Custom Tex Coord 2, sortPriority=-2}
  n2 = PreSkinnedNormal()
  FunctionOutput(None=AppendVector(
    A=MaterialFunctionCall(Float3=@n1) {materialFunction=&BreakOutFloat3Components}.B
    B=MaterialFunctionCall(Float3=@n2) {materialFunction=&BreakOutFloat3Components}.R
   )) {outputName=Custom Tex Coord 3, sortPriority=-3}
  FunctionOutput(None=ComponentMask(None=@n2) {b=true, g=true}) {outputName=Custom Tex Coord 4, sortPriority=-1}
  n3 = TextureCoordinate() {coordinateIndex=3}
  FunctionOutput(None=AppendVector(
    A=TextureCoordinate() {coordinateIndex=2}
    B=MaterialFunctionCall(Float2=@n3) {materialFunction=&BreakOutFloat2Components}.R
   )) {outputName=Weapon Local Position Pixel Shader, sortPriority=200}
  FunctionOutput(None=@n2) {outputName=Pre-Skinned Normal Vertex Normal, sortPriority=200}
  n4 = TextureCoordinate() {coordinateIndex=3}
  FunctionOutput(None=AppendVector(
    A=MaterialFunctionCall(Float2=@n4) {materialFunction=&BreakOutFloat2Components}.G
    B=TextureCoordinate() {coordinateIndex=4}
   )) {outputName=Pre-Skinned Normal Pixel Shader, sortPriority=200}
 ]
}
